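# M_BlackIce is a Nanite-enabled, Default Lit black ice material in the global MaterialGym folder.
# It uses material attributes, feeding a SetMaterialAttributes node with no shown inputs into the Ice_V2 material function borrowed from the weapon skin library.
# The function's Alpha is a constant one, and the visible graph holds no other logic, so the ice look comes from that function.
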Material /Game/Global/Materials/MaterialGym/Materials/BlackIce/M_BlackIce
graph {
 properties {bUseMaterialAttributes=true, bUsedWithNanite=true, pixelDepthOffsetMode=PDOM_Legacy, usedShadingModels=Default Lit}
 nodes [MaterialFunctionCall!(Alpha=Constant() {r=1}, In=SetMaterialAttributes!(MaterialAttributes)) {materialFunction=/Game/Weapons/WeaponSkins/MaterialFunction/Ice_V2}]
}
